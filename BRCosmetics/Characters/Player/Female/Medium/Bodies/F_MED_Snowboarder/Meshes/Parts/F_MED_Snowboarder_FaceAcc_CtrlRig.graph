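# Control Rig for the Snowboarder female face accessory that procedurally rotates the dyn_hatStrap_02 bone as the head turns.
# A DeltaRotation function swing-twist decomposes head rotation relative to head_null (under spine_05) and remaps one Euler axis to a clamped weight.
# That weight drives fixed strap rotation offsets in sequenced passes over different head-angle ranges.
# Two passes always run; a third runs only when Is Playing Emote is true.
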ControlRigBlueprint /BRCosmetics/Characters/Player/Female/Medium/Bodies/F_MED_Snowboarder/Meshes/Parts/F_MED_Snowboarder_FaceAcc_CtrlRig
graphs [
 New Function_ContainedGraph {nodes [
   n1 = RigVMFunctionEntryNode()
   RigVMFunction_MathBoolToInteger = &RigVMFunction_MathBoolToInteger(Value=@n1.onlyTwist)
   RigVMFunction_MathQuaternionSwingTwist = &RigVMFunction_MathQuaternionSwingTwist(
    TwistAxis=@n1.twistAxis
    Input=&RigUnit_GetRelativeTransformForItem(bChildInitial=False, bParentInitial=False, Child=@n1.child, Parent=@n1.parent).RelativeTransform.Rotation
   )
   n2 = RigVMDispatchNode(
    Values=((),())
    Index=@RigVMFunction_MathBoolToInteger.Result
    Values.0=@RigVMFunction_MathQuaternionSwingTwist.Swing
    Values.1=@RigVMFunction_MathQuaternionSwingTwist.Twist
   ) {template="DISPATCH_RigVMDispatch_SelectInt32(in Index,in Values,out Result)"}
   RigVMFunction_MathQuaternionToEuler = &RigVMFunction_MathQuaternionToEuler(Value=@n2.Result, RotationOrder=@n1.Rotation Order)
   > DISPATCH_RigVMDispatch_Print = RigVMDispatchNode(ScreenColor=(R=1,G=1,B=1,A=1), ScreenDuration=1.5, Enabled=@n1.debugRotation, Value=@RigVMFunction_MathQuaternionToEuler.Result) {template="DISPATCH_RigVMDispatch_Print(in Prefix,in Value,in Enabled,in ScreenDuration,in ScreenColor)"}
   Cast = RigVMDispatchNode(Value=@n1.triggerAxis) {template="DISPATCH_RigVMDispatch_CastEnumToInt(in Value,out Result)"}
   DISPATCH_RigVMDispatch_SelectInt32 = RigVMDispatchNode(
    Values=(0,0,0)
    Index=@Cast.Result
    Values.0=@RigVMFunction_MathQuaternionToEuler.Result.X
    Values.1=@RigVMFunction_MathQuaternionToEuler.Result.Y
    Values.2=@RigVMFunction_MathQuaternionToEuler.Result.Z
   ) {template="DISPATCH_RigVMDispatch_SelectInt32(in Index,in Values,out Result)"}
   Remap = &RigVMFunction_MathDoubleRemap(
    bClamp=False
    Value=@DISPATCH_RigVMDispatch_SelectInt32.Result
    TargetMinimum=@n1.outputMinMax.X
    TargetMaximum=@n1.outputMinMax.Y
    SourceMinimum=@n1.rotationMinMax.X
    SourceMaximum=@n1.rotationMinMax.Y
   )
   Absolute = &RigVMFunction_MathDoubleAbs(Value=@Remap.Result)
   Multiply = &RigVMFunction_MathDoubleMul(
    A=&RigVMFunction_MathDoubleClamp(Value=RigVMDispatchNode(True=@Absolute.Result, Condition=@n1.absoluteValue, False=@Remap.Result) {template="DISPATCH_RigVMDispatch_If(in Condition,in True,in False,out Result)"}.Result).Result
    B=@n1.premult
   )
   > RigVMFunctionReturnNode(outputQuat=@n2.Result, outputWeight=@Multiply.Result)
  ]}
 RigVMFunctionLibrary {nodes [DeltaRotation v4 = RigVMCollapseNode(
    ExecuteContext=()
    Rotation Order=ZYX
    child=(Type=Bone,Name="None")
    outputMinMax=(X=0,Y=1)
    parent=(Type=Bone,Name="None")
    premult=1
    rotationMinMax=(X=0,Y=0)
    twistAxis=(X=1,Y=0,Z=0)
   )]}
 RigVMFunction_Sequence_1_ContainedGraph {nodes [
   n1 = RigVMFunctionEntryNode()
   RigVMFunction_Sequence = &RigVMFunction_Sequence(ExecuteContext=@n1.ExecuteContext)
   n2 = &RigVMFunction_Sequence(ExecuteContext=@RigVMFunction_Sequence.B)
   n3 = &RigVMFunction_Sequence(ExecuteContext=@n2.B)
   n4 = &RigVMFunction_Sequence(ExecuteContext=@n3.B)
   RigVMFunctionReturnNode(A=@RigVMFunction_Sequence.A, B=@n2.A, C=@n3.A, D=@n4.A, E=@n4.B)
  ]}
 RigVMModel {nodes [
   RigUnit_BeginExecution = &RigUnit_BeginExecution()
   RigVMFunction_Sequence = RigVMAggregateNode(ExecuteContext=@RigUnit_BeginExecution.ExecutePin)
   &RigUnit_ModifyTransforms(
    ItemToModify=((Item=(Type=Bone,Name="head"),Transform=(Rotation=(X=0,Y=0,Z=0,W=1),Translation=(X=0,Y=0,Z=0),Scale3D=(X=1,Y=1,Z=1))))
    Weight=1
    WeightMinimum=0
    ExecutePin=@RigVMFunction_Sequence.B
   )
   > DeltaRotation v4 = RigVMFunctionReferenceNode(
    Rotation Order=ZYX
    child=(Type=Bone,Name="head")
    outputMinMax=(X=0,Y=1)
    parent=(Type=Null,Name="head_null")
    premult=1
    rotationMinMax=(X=-50,Y=-100)
    triggerAxis=Z
    twistAxis=(X=1,Y=0,Z=0)
   )
   > &RigUnit_ModifyTransforms(
    ItemToModify=((Item=(Type=Bone,Name="dyn_hatStrap_02"),Transform=(Rotation=(X=0,Y=0,Z=-0.342,W=0.9397),Translation=(X=0,Y=0,Z=0),Scale3D=(X=1,Y=1,Z=1))))
    WeightMinimum=0
    Weight=@"DeltaRotation v4".outputWeight
   )
   n1 = &RigVMFunction_ControlFlowBranch(ExecuteContext=@RigVMFunction_Sequence.C, Condition=RigVMVariableNode() {variable=Is Playing Emote}.Value)
   DeltaRotation v4_2 = RigVMFunctionReferenceNode(
    Rotation Order=ZYX
    child=(Type=Bone,Name="head")
    outputMinMax=(X=0,Y=1)
    parent=(Type=Null,Name="head_null")
    premult=1
    rotationMinMax=(X=-10,Y=-70)
    triggerAxis=Z
    twistAxis=(X=1,Y=0,Z=0)
    ExecuteContext=@n1.True
   )
   > &RigUnit_ModifyTransforms(
    ItemToModify=((Item=(Type=Bone,Name="dyn_hatStrap_02"),Transform=(Rotation=(X=0,Y=0,Z=-0.4226,W=0.9063),Translation=(X=0,Y=0,Z=0),Scale3D=(X=1,Y=1,Z=1))))
    WeightMinimum=0
    Weight=@"DeltaRotation v4_2".outputWeight
   )
   DeltaRotation v4_1 = RigVMFunctionReferenceNode(
    Rotation Order=ZYX
    child=(Type=Bone,Name="head")
    outputMinMax=(X=0,Y=1)
    parent=(Type=Null,Name="head_null")
    premult=1
    rotationMinMax=(X=-20,Y=-60)
    triggerAxis=Z
    twistAxis=(X=1,Y=0,Z=0)
    ExecuteContext=@RigVMFunction_Sequence.A
   )
   > &RigUnit_ModifyTransforms(
    ItemToModify=((Item=(Type=Bone,Name="dyn_hatStrap_02"),Transform=(Rotation=(X=0,Y=0,Z=-0.2588,W=0.9659),Translation=(X=0,Y=0,Z=0),Scale3D=(X=1,Y=1,Z=1))))
    WeightMinimum=0
    Weight=@"DeltaRotation v4_1".outputWeight
   )
  ]}
]
hierarchy [
 root {loc=<0 0 0>, rot=<0 0 0 1>}
 pelvis {parent=root, loc=<0 -0.5301 98.0571>, rot=<0.0127 0.707 -0.0127 -0.707>}
 spine_01 {parent=pelvis, loc=<6.1208 -0.9328 0>, rot=<0 0 0.034 -0.9994>}
 spine_02 {parent=spine_01, loc=<6.5997 -0.2095 0>, rot=<0 0 0.0238 -0.9997>}
 spine_03 {parent=spine_02, loc=<6.4914 -0.1237 0>, rot=<0 0 0.0149 -0.9999>}
 spine_04 {parent=spine_03, loc=<6.9598 0.0794 0>, rot=<0 0 -0.0132 -0.9999>}
 spine_05 {parent=spine_04, loc=<6.9987 0 0>, rot=<0 0 -0.0015 -1>}
 neck_01 {parent=spine_05, loc=<11.3138 0.1529 0>, rot=<0 0 -0.2003 -0.9797>}
 neck_02 {parent=neck_01, loc=<4.5627 0 0>, rot=<0 0 0.0504 -0.9987>}
 head {parent=neck_02, loc=<5.3914 0 0>, rot=<0 0 0.115 -0.9934>}
 dyn_pom {parent=head, loc=<13.8501 -2.4132 2.7198>, rot=<0 0 0.1775 -0.9841>}
 dyn_frontStrand_01 {parent=head, loc=<3.5382 8.5938 4.5569>, rot=<-0.0745 -0.9379 0.2797 0.1912>}
 dyn_frontStrand_01_end {parent=dyn_frontStrand_01, loc=<3.2664 0 0>, rot=<0 0 0 1>}
 dyn_frontStrand_02 {parent=head, loc=<0.2233 7.3725 5.866>, rot=<-0.0605 -0.9582 0.2765 0.0411>}
 dyn_frontStrand_02_end {parent=dyn_frontStrand_02, loc=<2.9018 0 0>, rot=<0 0 0 1>}
 dyn_frontStrand_03 {parent=head, loc=<3.4247 2.6873 -6.4094>, rot=<-0.0364 -0.9979 -0.0024 -0.0527>}
 dyn_frontStrand_03_end {parent=dyn_frontStrand_03, loc=<2.4634 0 0>, rot=<0 0 0 1>}
 dyn_frontStrand_04 {parent=head, loc=<4.4232 5.7675 -5.6289>, rot=<0.1782 -0.9552 -0.161 -0.1729>}
 dyn_frontStrand_05 {parent=head, loc=<5.307 9.7197 0.4781>, rot=<-0.1443 -0.7017 -0.6974 -0.0181>}
 dyn_bob_r_base {parent=head, loc=<-1.6833 1.6735 6.6658>, rot=<-0.0029 0.7071 0.7071 -0.0029>}
 dyn_bob_r_01 {parent=dyn_bob_r_base, loc=<0 0 0>, rot=<0 0 0 1>}
 dyn_bob_r_02 {parent=dyn_bob_r_01, loc=<4.4418 0 0>, rot=<0 0 0 1>}
 dyn_bob_r_03 {parent=dyn_bob_r_02, loc=<4.4418 0 0>, rot=<0 0 0 1>}
 dyn_bob_l_base {parent=head, loc=<-0.128 -1.9127 -6.1467>, rot=<-0.0029 0.7071 0.7071 -0.0029>}
 dyn_bob_l_01 {parent=dyn_bob_l_base, loc=<0 0 0>, rot=<0 0 0 1>}
 dyn_bob_l_02 {parent=dyn_bob_l_01, loc=<4.4418 0 0>, rot=<0 0 0 1>}
 dyn_bob_l_03 {parent=dyn_bob_l_02, loc=<4.4418 0 0>, rot=<0 0 0 1>}
 dyn_hatStrap_02_base {parent=head, loc=<5.3737 -5.3902 -5.9582>, rot=<0.6013 -0.741 -0.2809 0.1025>}
 dyn_hatStrap_02 {parent=dyn_hatStrap_02_base, loc=<0 0 0>, rot=<0 0 0 1>}
 dyn_hatStrap_02_end {parent=dyn_hatStrap_02, loc=<4.5564 0 0>, rot=<0 0 0 1>}
 dyn_hatStrap_01_base {parent=head, loc=<6.3922 -2.3586 -7.5826>, rot=<0.2232 -0.8431 -0.1594 -0.4626>}
 dyn_hatStrap_01 {parent=dyn_hatStrap_01_base, loc=<0 0 0>, rot=<0 0 0 1>}
 dyn_hatStrap_01_end {parent=dyn_hatStrap_01, loc=<3.9618 0 0>, rot=<0 0 0 1>}
 FACIAL_C_FacialRoot {parent=head, loc=<1.223 0.2837 0>, rot=<0.5021 -0.4979 -0.5021 -0.4979>}
 FACIAL_C_Forehead {parent=FACIAL_C_FacialRoot, loc=<0 -3.5996 7.9043>, rot=<-0.0436 0 0 -0.999>}
 FACIAL_L_ForeheadIn {parent=FACIAL_C_FacialRoot, loc=<1.6894 -3.3209 7.7185>, rot=<-0.0609 -0.0696 0.0043 -0.9957>}
 FACIAL_R_ForeheadIn {parent=FACIAL_C_FacialRoot, loc=<-1.6894 -3.3209 7.7185>, rot=<-0.0609 0.0696 -0.0043 -0.9957>}
 FACIAL_L_ForeheadMid {parent=FACIAL_C_FacialRoot, loc=<3.3079 -3.5819 7.152>, rot=<-0.0712 -0.1757 -0.0206 -0.9816>}
 FACIAL_R_ForeheadMid {parent=FACIAL_C_FacialRoot, loc=<-3.3079 -3.5819 7.152>, rot=<-0.0712 0.1757 0.0206 -0.9816>}
 FACIAL_L_ForeheadOut {parent=FACIAL_C_FacialRoot, loc=<4.7302 -3.6514 6.1056>, rot=<-0.1215 -0.2557 -0.0792 -0.9558>}
 FACIAL_R_ForeheadOut {parent=FACIAL_C_FacialRoot, loc=<-4.7302 -3.6514 6.1056>, rot=<-0.1215 0.2557 0.0792 -0.9558>}
 FACIAL_L_EyesackUpper {parent=FACIAL_C_FacialRoot, loc=<3.2542 -2.3917 7.6322>, rot=<0.0487 -0.0843 -0.0646 -0.9932>}
 FACIAL_R_EyesackUpper {parent=FACIAL_C_FacialRoot, loc=<-3.2542 -2.3917 7.6322>, rot=<0.0487 0.0843 0.0646 -0.9932>}
 FACIAL_L_EyelidUpperFurrow {parent=FACIAL_C_FacialRoot, loc=<3.2066 -2.211 7.4449>, rot=<-0.0301 -0.1509 -0.0797 -0.9849>}
 FACIAL_R_EyelidUpperFurrow {parent=FACIAL_C_FacialRoot, loc=<-3.2066 -2.211 7.4449>, rot=<-0.0301 0.1509 0.0797 -0.9849>}
 FACIAL_L_EyelidUpperA {parent=FACIAL_C_FacialRoot, loc=<2.9782 -1.0418 5.5653>, rot=<-0.1478 0 0 -0.989>}
 FACIAL_R_EyelidUpperA {parent=FACIAL_C_FacialRoot, loc=<-2.9782 -1.0418 5.5653>, rot=<-0.1478 0 0 -0.989>}
 FACIAL_L_Eye {parent=FACIAL_C_FacialRoot, loc=<2.9782 -1.0418 5.5653>, rot=<0 0 0 1>}
 FACIAL_R_Eye {parent=FACIAL_C_FacialRoot, loc=<-2.9782 -1.0418 5.5653>, rot=<0 0 0 1>}
 FACIAL_L_EyelidLowerA {parent=FACIAL_C_FacialRoot, loc=<2.9782 -1.0418 5.5653>, rot=<0.284 0 0 -0.9588>}
 FACIAL_R_EyelidLowerA {parent=FACIAL_C_FacialRoot, loc=<-2.9782 -1.0418 5.5653>, rot=<0.284 0 0 -0.9588>}
 FACIAL_L_EyeCornerInner {parent=FACIAL_C_FacialRoot, loc=<1.6619 -0.5814 6.5669>, rot=<0 0 0 -1>}
 FACIAL_R_EyeCornerInner {parent=FACIAL_C_FacialRoot, loc=<-1.6619 -0.5814 6.5669>, rot=<0 0 0 -1>}
 FACIAL_L_EyeCornerOuter {parent=FACIAL_C_FacialRoot, loc=<4.5502 -0.9745 5.9071>, rot=<0 -0.3827 0 -0.9239>}
 FACIAL_R_EyeCornerOuter {parent=FACIAL_C_FacialRoot, loc=<-4.5502 -0.9745 5.9071>, rot=<0 0.3827 0 -0.9239>}
 FACIAL_L_EyesackLower {parent=FACIAL_C_FacialRoot, loc=<3.706 0.342 6.3213>, rot=<-0.0223 -0.1398 -0.0223 -0.9897>}
 FACIAL_R_EyesackLower {parent=FACIAL_C_FacialRoot, loc=<-3.706 0.342 6.3213>, rot=<-0.0223 0.1398 0.0223 -0.9897>}
 FACIAL_L_CheekInner {parent=FACIAL_C_FacialRoot, loc=<2.3426 1.0496 6.6479>, rot=<-0.1233 -0.088 -0.1972 -0.9686>}
 FACIAL_R_CheekInner {parent=FACIAL_C_FacialRoot, loc=<-2.3426 1.0496 6.6479>, rot=<-0.1233 0.088 0.1972 -0.9686>}
 FACIAL_L_CheekOuter {parent=FACIAL_C_FacialRoot, loc=<3.9986 1.2984 4.857>, rot=<0.0551 -0.377 0.1003 -0.9191>}
 FACIAL_R_CheekOuter {parent=FACIAL_C_FacialRoot, loc=<-3.9986 1.2984 4.857>, rot=<0.0551 0.377 -0.1003 -0.9191>}
 FACIAL_L_NasolabialBulge {parent=FACIAL_C_FacialRoot, loc=<2.9322 4.0534 5.0746>, rot=<0.0579 -0.3774 0.0238 -0.9239>}
 FACIAL_R_NasolabialBulge {parent=FACIAL_C_FacialRoot, loc=<-2.9322 4.0534 5.0746>, rot=<0.0579 0.3774 -0.0238 -0.9239>}
 FACIAL_L_NasolabialFurrow {parent=FACIAL_C_FacialRoot, loc=<2.7397 3.2062 7.4914>, rot=<-0.0333 -0.1258 -0.3638 -0.9223>}
 FACIAL_R_NasolabialFurrow {parent=FACIAL_C_FacialRoot, loc=<-2.7397 3.2062 7.4914>, rot=<-0.0333 0.1258 0.3638 -0.9223>}
 FACIAL_L_CheekLower {parent=FACIAL_C_FacialRoot, loc=<2.8516 4.3553 5.3095>, rot=<0.0345 -0.427 0.0577 -0.9018>}
 FACIAL_R_CheekLower {parent=FACIAL_C_FacialRoot, loc=<-2.8516 4.3553 5.3095>, rot=<0.0345 0.427 -0.0577 -0.9018>}
 FACIAL_C_Nose {parent=FACIAL_C_FacialRoot, loc=<0 -0.0416 8.0332>, rot=<-0.1305 0 0 -0.9914>}
 FACIAL_C_NoseLower {parent=FACIAL_C_Nose, loc=<0 3.1434 -0.0221>, rot=<0.1305 0 0 -0.9914>}
 FACIAL_L_Nostril {parent=FACIAL_C_Nose, loc=<0.7104 2.2138 -0.0586>, rot=<0.0173 0.1305 0.0023 -0.9913>}
 FACIAL_R_Nostril {parent=FACIAL_C_Nose, loc=<-0.7104 2.2138 -0.0586>, rot=<0.0173 -0.1305 -0.0023 -0.9913>}
 FACIAL_C_NoseTip {parent=FACIAL_C_Nose, loc=<0.0008 2.3312 0.769>, rot=<0.0872 0 0 -0.9962>}
 FACIAL_C_MouthUpper {parent=FACIAL_C_FacialRoot, loc=<0.0001 4.5944 6.2212>, rot=<0 0 0 -1>}
 FACIAL_C_LipUpper {parent=FACIAL_C_MouthUpper, loc=<-0.0001 -0.3125 2.0438>, rot=<-0.0872 0 0 -0.9962>}
 FACIAL_L_LipUpper {parent=FACIAL_C_MouthUpper, loc=<0.719 -0.3732 1.9975>, rot=<-0.0778 -0.1601 -0.0379 -0.9833>}
 FACIAL_R_LipUpper {parent=FACIAL_C_MouthUpper, loc=<-0.719 -0.3732 1.9975>, rot=<-0.0778 0.1601 0.0379 -0.9833>}
 FACIAL_L_LipUpperOuter {parent=FACIAL_C_MouthUpper, loc=<1.5571 -0.2384 1.4191>, rot=<-0.0983 -0.2294 -0.0983 -0.9634>}
 FACIAL_R_LipUpperOuter {parent=FACIAL_C_MouthUpper, loc=<-1.5571 -0.2384 1.4191>, rot=<-0.0983 0.2294 0.0983 -0.9634>}
 FACIAL_L_LipCorner {parent=FACIAL_C_MouthUpper, loc=<2.3191 0.1603 0.5958>, rot=<0 -0.2079 0 -0.9781>}
 FACIAL_R_LipCorner {parent=FACIAL_C_MouthUpper, loc=<-2.3191 0.1603 0.5958>, rot=<0 0.2079 0 -0.9781>}
 FACIAL_L_Masseter {parent=FACIAL_C_FacialRoot, loc=<5.9821 1.9533 2.9079>, rot=<0.0693 0.1129 0.0079 -0.9912>}
 FACIAL_R_Masseter {parent=FACIAL_C_FacialRoot, loc=<-5.9821 1.9533 2.9079>, rot=<0.0693 -0.1129 -0.0079 -0.9912>}
 FACIAL_C_UnderChin {parent=FACIAL_C_FacialRoot, loc=<0 8.2231 3.1919>, rot=<0.5 0 0 -0.866>}
 FACIAL_L_UnderChin {parent=FACIAL_C_FacialRoot, loc=<2.6353 7.2209 1.8592>, rot=<0.4025 -0.355 0.0823 -0.8398>}
 FACIAL_R_UnderChin {parent=FACIAL_C_FacialRoot, loc=<-2.6353 7.2209 1.8592>, rot=<0.4025 0.355 -0.0823 -0.8398>}
 FACIAL_C_TeethUpper {parent=FACIAL_C_FacialRoot, loc=<0 4.5889 7.3455>, rot=<0 0 0 1>}
 FACIAL_C_LowerLipRotation {parent=FACIAL_C_FacialRoot, loc=<0.0001 1.9224 1.0678>, rot=<0.342 0 0 -0.9397>}
 FACIAL_C_MouthLower {parent=FACIAL_C_LowerLipRotation, loc=<0 -1.2656 5.6653>, rot=<-0.342 0 0 -0.9397>}
 FACIAL_C_LipLower {parent=FACIAL_C_MouthLower, loc=<-0.0001 0.6577 1.8587>, rot=<0.0872 0 0 -0.9962>}
 FACIAL_L_LipLower {parent=FACIAL_C_MouthLower, loc=<0.6872 0.6171 1.7334>, rot=<0.0927 -0.1295 0.0273 -0.9869>}
 FACIAL_R_LipLower {parent=FACIAL_C_MouthLower, loc=<-0.6872 0.6171 1.7334>, rot=<0.0927 0.1295 -0.0273 -0.9869>}
 FACIAL_L_LipLowerOuter {parent=FACIAL_C_MouthLower, loc=<1.5399 0.5262 1.1451>, rot=<0.0913 -0.2322 0.0318 -0.9679>}
 FACIAL_R_LipLowerOuter {parent=FACIAL_C_MouthLower, loc=<-1.5399 0.5262 1.1451>, rot=<0.0913 0.2322 -0.0318 -0.9679>}
 FACIAL_C_TeethLower {parent=FACIAL_C_LowerLipRotation, loc=<-0.0001 -1.9286 6.4717>, rot=<-0.342 0 0 -0.9397>}
 FACIAL_C_Tongue1 {parent=FACIAL_C_TeethLower, loc=<0 0.3587 -4.85>, rot=<-0.2588 0 0 -0.9659>}
 FACIAL_C_Tongue2 {parent=FACIAL_C_Tongue1, loc=<0 0.3451 1.4312>, rot=<0.2588 0 0 -0.9659>}
 FACIAL_C_Tongue3 {parent=FACIAL_C_Tongue2, loc=<0 0.1493 1.2843>, rot=<0.1305 0 0 -0.9914>}
 FACIAL_C_Tongue4 {parent=FACIAL_C_Tongue3, loc=<0.0006 -0.2194 1.0685>, rot=<0.0436 0 0 -0.999>}
 FACIAL_C_TongueUpper3 {parent=FACIAL_C_Tongue3, loc=<0 -0.2941 -0.2398>, rot=<0 0 0 1>}
 FACIAL_C_TongueLower3 {parent=FACIAL_C_Tongue3, loc=<0 0.8294 0.0462>, rot=<0 0 0 1>}
 FACIAL_L_TongueSide3 {parent=FACIAL_C_Tongue3, loc=<1.4941 -0.0273 0.0614>, rot=<0 0 0 1>}
 FACIAL_R_TongueSide3 {parent=FACIAL_C_Tongue3, loc=<-1.4942 -0.0273 0.0614>, rot=<0 0 0 1>}
 FACIAL_C_TongueUpper2 {parent=FACIAL_C_Tongue2, loc=<0 -0.5618 -0.3851>, rot=<0 0 0 1>}
 FACIAL_L_TongueSide2 {parent=FACIAL_C_Tongue2, loc=<1.7814 -0.0872 -0.0014>, rot=<0 0 0 1>}
 FACIAL_R_TongueSide2 {parent=FACIAL_C_Tongue2, loc=<-1.7814 -0.0872 -0.0014>, rot=<0 0 0 1>}
 FACIAL_C_Jaw {parent=FACIAL_C_FacialRoot, loc=<0.0001 1.9224 1.0678>, rot=<0.342 0 0 -0.9397>}
 FACIAL_C_Jawline {parent=FACIAL_C_Jaw, loc=<-0.0001 1.8242 7.4454>, rot=<0.2588 0 0 -0.9659>}
 FACIAL_L_ChinSide {parent=FACIAL_C_Jaw, loc=<2.4041 0.1474 7.2689>, rot=<-0.2207 -0.3739 0.05 -0.8994>}
 FACIAL_R_ChinSide {parent=FACIAL_C_Jaw, loc=<-2.4041 0.1474 7.2689>, rot=<-0.2207 0.3739 -0.05 -0.8994>}
 FACIAL_C_Chin {parent=FACIAL_C_Jaw, loc=<-0.0001 0.1689 6.7547>, rot=<-0.2588 0 0 -0.9659>}
 FACIAL_L_Ear {parent=FACIAL_C_FacialRoot, loc=<6.7509 -0.2979 -0.4917>, rot=<-0.0262 -0.0007 -0.0262 -0.9993>}
 FACIAL_R_Ear {parent=FACIAL_C_FacialRoot, loc=<-6.7509 -0.2979 -0.4917>, rot=<-0.0262 0.0007 0.0262 -0.9993>}
 FACIAL_L_Jawline {parent=FACIAL_C_FacialRoot, loc=<3.9799 5.6364 3.3481>, rot=<0.2092 0.1578 -0.0547 -0.9635>}
 FACIAL_R_Jawline {parent=FACIAL_C_FacialRoot, loc=<-3.9799 5.6364 3.3481>, rot=<0.2092 -0.1578 0.0547 -0.9635>}
 FACIAL_C_NoseBridge {parent=FACIAL_C_FacialRoot, loc=<-0.0009 -1.8025 8.6583>, rot=<0.0436 0 0 -0.999>}
 FACIAL_L_EyelidUpperB {parent=FACIAL_C_FacialRoot, loc=<2.9782 -1.0418 5.5653>, rot=<-0.225 0 0 -0.9744>}
 FACIAL_L_EyelidLowerB {parent=FACIAL_C_FacialRoot, loc=<2.9782 -1.0418 5.5653>, rot=<0.3173 0 0 -0.9483>}
 FACIAL_R_EyelidLowerB {parent=FACIAL_C_FacialRoot, loc=<-2.9782 -1.0418 5.5653>, rot=<0.3256 0 0 -0.9455>}
 FACIAL_R_EyelidUpperB {parent=FACIAL_C_FacialRoot, loc=<-2.9782 -1.0418 5.5653>, rot=<-0.225 0 0 -0.9744>}
 dyn_braid_base {parent=head, loc=<1.014 4.7073 6.7068>, rot=<0.0168 -0.7119 -0.6988 -0.0678>}
 dyn_braid_01 {parent=dyn_braid_base, loc=<0 0 0>, rot=<0.0088 0.0003 -0.0383 -0.9992>}
 dyn_braid_02 {parent=dyn_braid_01, loc=<3.737 0 0>, rot=<-0.0088 -0.001 0.1076 -0.9942>}
 dyn_braid_03 {parent=dyn_braid_02, loc=<4.7097 0 0>, rot=<0 0 0.0633 -0.998>}
 dyn_braid_04 {parent=dyn_braid_03, loc=<5.0186 0 0>, rot=<0 0 0 1>}
 dyn_neck_noTwist {parent=spine_05, loc=<11.3138 0.1529 0>, rot=<0 0 -0.2003 -0.9797>}
 dyn_c_collar_base {parent=dyn_neck_noTwist, loc=<-0.9121 -5.6344 0>, rot=<0 0 0.1608 -0.987>}
 dyn_c_collar {parent=dyn_c_collar_base, loc=<0 0 0>, rot=<0 0 0 1>}
 dyn_r_collar_base {parent=spine_05, loc=<8.1084 3.5527 9.3467>, rot=<0 0 0.9712 -0.2381>}
 dyn_r_collar {parent=dyn_r_collar_base, loc=<0 0 0>, rot=<0 0 0 1>}
 dyn_l_collar_base {parent=spine_05, loc=<8.1083 3.5527 -9.3467>, rot=<0 0 -0.2381 -0.9712>}
 dyn_l_collar {parent=dyn_l_collar_base, loc=<0 0 0>, rot=<0 0 0 1>}
 dyn_l_collarSide_base {parent=spine_05, loc=<10.5494 -3.6074 -8.0786>, rot=<0 0 -0.1398 -0.9902>}
 dyn_l_collarSide {parent=dyn_l_collarSide_base, loc=<0 0 0>, rot=<0 0 0 1>}
 dyn_r_collarSide_base {parent=spine_05, loc=<10.549 -3.6075 8.0786>, rot=<0 0 0.9902 -0.1398>}
 dyn_r_collarSide {parent=dyn_r_collarSide_base, loc=<0 0 0>, rot=<0 0 0 -1>}
 head_null {parent=spine_05, type=Null, loc=<20.6573 3.5489 0>, rot=<0 0 -0.036 -0.9994>}
]
variables [Is Playing Emote {type=bool, default=False}]
